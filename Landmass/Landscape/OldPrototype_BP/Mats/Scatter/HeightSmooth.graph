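# Unlit alpha-composite material from an old Landmass landscape prototype that smooths the blue channel of a height render target.
# Four texel-offset neighbor taps are summed, with the bottom tap reading VelocityRT rather than HeightRT.
# A five-tap average is lerped against the center and packed with red duplicated in X and Y.
# An alternative weighted Laplacian-style blend also exists, and no material output connections appear.
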Material /Landmass/Landscape/OldPrototype_BP/Mats/Scatter/HeightSmooth
graph {
 properties {
  bAllowNegativeEmissiveColor=true
  blendMode=BLEND_AlphaComposite
  pixelDepthOffsetMode=PDOM_Legacy
  shadingModel=MSM_Unlit
  usedShadingModels=Unlit
 }
 nodes [
  n1 = TextureSampleParameter2D() {parameterName=HeightRT, texture=&Black}
  n2 = ScalarParameter() {defaultValue=0.1, parameterName=a}
  n3 = Multiply(A=Constant() {r=4}, B=@n2)
  n4 = TextureCoordinate()
  n5 = AppendVector(A=ScalarParameter() {defaultValue=2048, parameterName=SizeX}, B=ScalarParameter() {defaultValue=2048, parameterName=SizeY})
  n6 = Multiply(A=Divide(B=@n5))
  n7 = Multiply(A=@n6, B=Constant2Vector() {r=1})
  n8 = Subtract(A=@n4, B=@n7)
  n9 = Add(A=@n4, B=@n7)
  n10 = TextureCoordinate()
  n11 = Multiply(A=@n6, B=Constant2Vector() {g=1})
  n12 = Subtract(A=@n10, B=@n11)
  n13 = Add(A=@n10, B=@n11)
  n14 = Add(
   A=Add(
    A=TextureSampleParameter2D(UVs=@n8) {desc=Left Wall, parameterName=HeightRT, texture=&Black}.B
    B=TextureSampleParameter2D(UVs=@n9) {desc=Right Wall, parameterName=HeightRT, texture=&Black}.B
   )
   B=Add(
    A=TextureSampleParameter2D(UVs=@n12) {desc=Bottom Wall, parameterName=VelocityRT, texture=&Black}.B
    B=TextureSampleParameter2D(UVs=@n13) {desc=Top Wall, parameterName=HeightRT, texture=&Black}.B
   )
  )
  Add(A=Multiply(A=@n1.B, B=OneMinus(None=@n3)), B=Multiply(A=@n2, B=@n14))
  n15 = Add(A=@n1.B, B=@n14)
  MaterialFunctionCall(X=@n1.R, Y=@n1.R, Z=LinearInterpolate(A=@n1.B, B=Divide(A=@n15) {constB=5}, Alpha=ScalarParameter())) {materialFunction=&MakeFloat3}
  Multiply(A=@n5, B=ScalarParameter() {defaultValue=16, parameterName=GridSize})
  OneMinus(None=TextureSampleParameter2D() {parameterName=HeightRT, texture=&Black}.A)
  ScalarParameter() {defaultValue=1, parameterName=deltaT}
  ScalarParameter()
  TextureCoordinate()
 ]
}
