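# Prototype Landmass material that previews erosion scatter particles as additive, unlit, two-sided quads on instanced meshes.
# Per-instance coordinates sample position and velocity render targets, with MPC_Erosion grid values setting placement.
# World position offset math builds camera-facing quads stretched along velocity, with optional time-animated movement.
# Color math combines velocity, blue tint and camera-distance-faded debug digits; float16-unpacking Custom nodes lead only to unconsumed vectors.
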Material /Landmass/Landscape/OldPrototype_BP/Mats/Scatter/ParticlePreview
graph {
 properties {
  bTangentSpaceNormal=false
  bUsedWithInstancedStaticMeshes=true
  blendMode=BLEND_Additive
  pixelDepthOffsetMode=PDOM_Legacy
  shadingModel=MSM_Unlit
  twoSided=true
  usedShadingModels=Unlit
 }
 nodes [
  Add(A=Multiply(A=WorldPosition().XYZ, B=Constant3Vector() {constant=<rgba 0 0 -1 0>}))
  n1 = ScalarParameter() {defaultValue=1000, parameterName=TextMax}
  n2 = MaterialFunctionCall(Fade Length=@n1, Fade Offset=@n1) {materialFunction=&CameraDepthFade}
  n3 = VertexInterpolator(VS=MaterialFunctionCall() {materialFunction=/Landmass/Landscape/OldPrototype_BP/Mats/MF/Scatter_InstanceCoords}.Result)
  n4 = TextureSampleParameter2D(UVs=@n3.PS) {parameterName=PositionRT, texture=&Black}
  n5 = TextureSampleParameter2D(UVs=@n3.PS) {parameterName=VelocityRT, texture=&Black}
  n6 = TextureSampleParameter2D(UVs=@n3.PS) {parameterName=VelocityRTPrev, texture=&Black}
  n7 = TextureSampleParameter2D(UVs=@n3.PS) {parameterName=HeightRT, texture=&Black}
  n8 = TextureCoordinate()
  n9 = MaterialFunctionCall!(
   Vector4=MaterialFunctionCall(X=Multiply(A=@n4.A) {constB=1000}, Y=@n5.A, Z=Subtract(A=@n5.A, B=@n6.A), A=Subtract(A=@n4.B, B=@n7.R)) {materialFunction=&MakeFloat4}.Result
   MaximumNumberOfDigits=ScalarParameter() {defaultValue=7, parameterName=dddd}
   UVs=@n8
   DebugTextLocation RG_UpperRight BA_LowerLeft
   Component Spacing
  ) {materialFunction=/Engine/Functions/Engine_MaterialFunctions02/Utility/DebugFloat4Values}
  n10 = Add(A=ComponentMask(None=@n9.ColorCodedOutput) {b=true, g=true, r=true}, B=ComponentMask(None=@n9.ColorCodedOutput) {a=true})
  n11 = MaterialFunctionCall() {materialFunction=&Scatter_InstanceCoords}
  n12 = VertexInterpolator(VS=@n11.Result)
  n13 = TextureSampleParameter2D(UVs=@n12.PS) {parameterName=VelocityRT, texture=&Black}
  n14 = CollectionParameter() {collection=/Landmass/Landscape/BlueprintBrushes/MPC/MPC_Erosion, parameterName=GridSize}
  n15 = Multiply(A=Abs(None=Divide(A=@n13.RGB, B=@n14)), B=Divide(A=ScalarParameter() {defaultValue=1, parameterName=Brightness}) {constB=16})
  n16 = CollectionParameter() {collection=&MPC_Erosion, parameterName=VelocityDisplay}
  n17 = Constant3Vector() {constant=<rgba 0 0 0.4 0>}
  n18 = ScalarParameter() {defaultValue=0.1, parameterName=BlueEmissive}
  n19 = Multiply(A=Add(A=Multiply(A=@n15, B=@n16), B=Multiply(A=@n17, B=@n18)))
  Add(A=Multiply(A=OneMinus(None=@n2.Result), B=Multiply(A=@n10) {constB=10}), B=@n19)
  AppendVector()
  Constant()
  Constant() {r=1}
  MaterialFunctionCall!(
   Vector2=VertexInterpolator(VS=MaterialFunctionCall() {materialFunction=&Scatter_InstanceCoords}.Result).PS
   MaximumNumberOfDigits=Constant() {r=8}
   UVs=TextureCoordinate()
   DebugTextLocation RG_UpperRight BA_LowerLeft
   Component Spacing
  ) {materialFunction=/Engine/Functions/Engine_MaterialFunctions02/Utility/DebugFloat2Values}
  MaterialExpressionTextureSampleParameter2D_13 = TextureSampleParameter2D(UVs=@n11.Result) {parameterName=VelocityRT, texture=&Black}
  n20 = Custom!(N=@MaterialExpressionTextureSampleParameter2D_13.R) {
   code="//extracts 2 16bit floats from a  32bit float\r\n\r\nuint f16 = asuint(N);\r\nf16 = (f16 && 0xffff0000) >> 16;\r\nuint sign = (f16 && 0x8000)<<16;\r\nuint mant = (f16 && 0x3ffu) <<13; // 13 bits are zeros\r\nuint exp =  (f16&&0x7fff)==0 ? 0 : (((((const uint) f16)>>10)&&0x1f));\r\nf16 =  sign | (exp -15+127) << 23 | mant;\r\nfloat outval1 = asfloat( f16 );\r\n\r\nf16 = asuint(N);\r\nf16 = (f16 && 0xffff);\r\nsign = (f16 && 0x8000)<<16;\r\nmant = (f16 && 0x3ffu) <<13; // 13 bits are zeros\r\nexp =  (f16&&0x7fff)==0 ? 0 : (((((const uint) f16)>>10)&&0x1f));\r\nf16 =  sign | (exp -15+127) << 23 | mant;\r\nfloat outval2 = asfloat( f16 );\r\n\r\nreturn float2(outval1, outval2);\r\n"
   description=Unpack 2 f16s from f32
   outputType=CMOT_Float2
   inputs [{
     inputName=N
     input {
      expression=&self:MaterialExpressionTextureSampleParameter2D_13
      mask=1
      maskR=1
      outputIndex=1
     }
    }]
  }
  n21 = MaterialFunctionCall(Float2=@n20) {materialFunction=&BreakOutFloat2Components}
  n22 = Custom(N=@MaterialExpressionTextureSampleParameter2D_13.G) {
   code="//extracts 2 16bit floats from a  32bit float\r\n\r\nuint f16 = asuint(N);\r\nf16 = (f16 && 0xffff0000) >> 16;\r\nuint sign = (f16 && 0x8000)<<16;\r\nuint mant = (f16 && 0x3ffu) <<13; // 13 bits are zeros\r\nuint exp =  (f16&&0x7fff)==0 ? 0 : (((((const uint) f16)>>10)&&0x1f));\r\nf16 =  sign | (exp -15+127) << 23 | mant;\r\nfloat outval1 = asfloat( f16 );\r\n\r\nf16 = asuint(N);\r\nf16 = (f16 && 0xffff);\r\nsign = (f16 && 0x8000)<<16;\r\nmant = (f16 && 0x3ffu) <<13; // 13 bits are zeros\r\nexp =  (f16&&0x7fff)==0 ? 0 : (((((const uint) f16)>>10)&&0x1f));\r\nf16 =  sign | (exp -15+127) << 23 | mant;\r\nfloat outval2 = asfloat( f16 );\r\n\r\nreturn float2(outval1, outval2);\r\n"
   description=Unpack 2 f16s from f32
   outputType=CMOT_Float2
   inputs [{
     inputName=N
     input {
      expression=&self:MaterialExpressionTextureSampleParameter2D_13
      mask=1
      maskG=1
      outputIndex=2
     }
    }]
  }
  n23 = MaterialFunctionCall(Float2=@n22) {materialFunction=&BreakOutFloat2Components}
  n24 = Custom(N=@MaterialExpressionTextureSampleParameter2D_13.B) {
   code="//extracts 2 16bit floats from a  32bit float\r\n\r\nuint f16 = asuint(N);\r\nf16 = (f16 && 0xffff0000) >> 16;\r\nuint sign = (f16 && 0x8000)<<16;\r\nuint mant = (f16 && 0x3ffu) <<13; // 13 bits are zeros\r\nuint exp =  (f16&&0x7fff)==0 ? 0 : (((((const uint) f16)>>10)&&0x1f));\r\nf16 =  sign | (exp -15+127) << 23 | mant;\r\nfloat outval1 = asfloat( f16 );\r\n\r\nf16 = asuint(N);\r\nf16 = (f16 && 0xffff);\r\nsign = (f16 && 0x8000)<<16;\r\nmant = (f16 && 0x3ffu) <<13; // 13 bits are zeros\r\nexp =  (f16&&0x7fff)==0 ? 0 : (((((const uint) f16)>>10)&&0x1f));\r\nf16 =  sign | (exp -15+127) << 23 | mant;\r\nfloat outval2 = asfloat( f16 );\r\n\r\nreturn float2(outval1, outval2);\r\n"
   description=Unpack 2 f16s from f32
   outputType=CMOT_Float2
   inputs [{
     inputName=N
     input {
      expression=&self:MaterialExpressionTextureSampleParameter2D_13
      mask=1
      maskB=1
      outputIndex=3
     }
    }]
  }
  n25 = MaterialFunctionCall(Float2=@n24) {materialFunction=&BreakOutFloat2Components}
  MaterialFunctionCall(X=@n21.R, Y=@n23.R, Z=@n25.R) {desc=Velocity, materialFunction=&MakeFloat3}
  MaterialFunctionCall(X=@n21.G, Y=@n23.G, Z=@n25.G) {desc=Position, materialFunction=&MakeFloat3}
  MaterialFunctionCall(Vector 3=@n5.RGB) {materialFunction=/Engine/Functions/Engine_MaterialFunctions02/Utility/VectorLength}
  MaterialFunctionCall(Vector 3=@n6.RGB) {materialFunction=&VectorLength}
  n26 = TextureSampleParameter2D(UVs=@n12.PS) {parameterName=PositionRT, texture=&Black}
  Multiply(A=@n19, B=Saturate(None=Multiply(A=@n26.A) {constB=1000}))
  OneMinus(None=MaterialFunctionCall(Float2=@n8) {materialFunction=&BreakOutFloat2Components}.G)
  n27 = Multiply(A=WorldPosition() {worldPositionShaderOffset=WPT_ExcludeAllShaderOffsets}.XYZ) {constB=-1}
  n28 = ConstantBiasScale(None=@n11.Result) {bias=-0.5, scale=1}
  n29 = Multiply(
   A=AppendVector(
    A=CollectionParameter() {collection=&MPC_Erosion, parameterName=ResX}
    B=CollectionParameter() {collection=&MPC_Erosion, parameterName=ResY}
   )
   B=CollectionParameter() {collection=&MPC_Erosion, parameterName=GridSize}
  )
  n30 = Multiply(
   A=CollectionParameter() {collection=&MPC_Erosion, parameterName=deltaT}
   B=@MaterialExpressionTextureSampleParameter2D_13.RGB
  )
  n31 = Multiply(A=ScalarParameter() {parameterName=show delta}, B=Frac(None=Time()))
  n32 = Add(
   A=AppendVector(A=Multiply(A=@n28, B=@n29), B=ScalarParameter() {defaultValue=50, parameterName=PosZ})
   B=Add(
    A=TextureSampleParameter2D(UVs=@n11.Result) {parameterName=PositionRT, texture=&Black}.RGB
    B=Multiply(A=@n30, B=@n31)
   )
  )
  n33 = MaterialFunctionCall(Vector 3=Multiply(A=@MaterialExpressionTextureSampleParameter2D_13.RGB, B=ScalarParameter() {defaultValue=0.2, parameterName=MV Amount})) {materialFunction=&VectorLength}
  n34 = ScalarParameter() {defaultValue=100, parameterName=Size}
  n35 = Normalize(VectorInput=Add(A=@MaterialExpressionTextureSampleParameter2D_13.RGB, B=VectorParameter() {parameterName=fsdfsf, defaultValue=<rgba 0.001 0 0 0>}.RGB))
  n36 = ConstantBiasScale(None=TextureCoordinate()) {bias=-0.5, scale=1}
  n37 = Multiply(A=Multiply(A=Add(A=@n33.V3 Length, B=@n34), B=@n35), B=Multiply(A=ComponentMask(None=@n36) {r=true}) {constB=-1}) {desc=X}
  n38 = CameraPositionWS()
  n39 = Multiply(A=Normalize(VectorInput=Subtract(A=@n32, B=@n38))) {constB=-1}
  n40 = Multiply(
   A=Multiply(A=ComponentMask(None=@n36) {g=true}, B=ScalarParameter() {defaultValue=100, parameterName=Size}) {desc=Y}
   B=Normalize(VectorInput=CrossProduct(A=@n35, B=@n39))
  )
  Reroute(None=Add(A=Add(A=@n27, B=@n32), B=Add(A=@n37, B=@n40)))
  ScalarParameter() {defaultValue=-0.09}
  TextureCoordinate() {coordinateIndex=1}
  TextureCoordinate() {coordinateIndex=1}
  TextureCoordinate() {coordinateIndex=1}
 ]
}
