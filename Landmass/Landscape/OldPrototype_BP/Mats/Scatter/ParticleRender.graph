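# Additive unlit material for instanced meshes that scatters erosion-simulation particles, apparently to splat their values onto the grid.
# World position offset places each instance's quad using PositionRT, optionally snapped and sized to MPC_Erosion grid cells.
# Emissive carries scaled velocity in RG and a saturation-change plus evaporation term in B, weighted by sub-cell position when quantized.
# Negative emissive is allowed, so additive blending can also subtract; leftover debug nodes remain.
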Material /Landmass/Landscape/OldPrototype_BP/Mats/Scatter/ParticleRender
graph {
 properties {
  bAllowNegativeEmissiveColor=true
  bTangentSpaceNormal=false
  bUsedWithInstancedStaticMeshes=true
  blendMode=BLEND_Additive
  pixelDepthOffsetMode=PDOM_Legacy
  shadingModel=MSM_Unlit
  twoSided=true
  usedShadingModels=Unlit
 }
 nodes [
  n1 = WorldPosition() {worldPositionShaderOffset=WPT_ExcludeAllShaderOffsets}
  n2 = TextureSampleParameter2D(UVs=MaterialFunctionCall() {materialFunction=/Landmass/Landscape/OldPrototype_BP/Mats/MF/Scatter_InstanceCoords}.Result) {parameterName=PositionRT, texture=&Black}
  n3 = ComponentMask(None=@n2.RGB) {g=true, r=true}
  n4 = CollectionParameter() {collection=/Landmass/Landscape/BlueprintBrushes/MPC/MPC_Erosion, parameterName=GridSize}
  n5 = Multiply(A=Floor(None=Divide(A=@n3, B=@n4)), B=@n4)
  n6 = ScalarParameter() {defaultValue=1, parameterName=Quantize, sliderMax=1}
  n7 = Multiply(A=AppendVector(A=LinearInterpolate(A=@n3, B=@n5, Alpha=@n6), B=@n2.B))
  n8 = ConstantBiasScale(None=MaterialFunctionCall() {materialFunction=&Scatter_InstanceCoords}.Result) {bias=-0.5, scale=1}
  n9 = CollectionParameter() {collection=&MPC_Erosion, parameterName=GridSize}
  n10 = Multiply(
   A=AppendVector(
    A=CollectionParameter() {collection=&MPC_Erosion, parameterName=ResX}
    B=CollectionParameter() {collection=&MPC_Erosion, parameterName=ResY}
   )
   B=@n9
  )
  n11 = ScalarParameter() {defaultValue=0.5, parameterName=CenterOnTexel}
  n12 = AppendVector(
   A=Add(A=Multiply(A=@n8, B=@n10), B=Multiply(A=@n9, B=@n11)) {desc=Half texel offset}
   B=ScalarParameter() {defaultValue=16, parameterName=Zoffset}
  )
  n13 = ComponentMask(None=ConstantBiasScale(None=TextureCoordinate()) {bias=-0.5, scale=1}) {g=true, r=true}
  n14 = Multiply(
   A=CollectionParameter() {collection=&MPC_Erosion, parameterName=GridSize}
   B=LinearInterpolate(B=ScalarParameter() {defaultValue=2, parameterName=Size}, Alpha=ScalarParameter() {defaultValue=1, parameterName=Quantize, sliderMax=1}) {constA=1}
  )
  Add(A=Add(A=Multiply(A=@n1.XYZ) {constB=-1}, B=Add(A=@n7, B=@n12)), B=AppendVector(A=Multiply(A=@n13, B=@n14), B=Constant()))
  n15 = TextureCoordinate()
  n16 = Frac(None=OneMinus(None=Multiply(A=@n15) {constB=2}))
  n17 = VertexInterpolator(VS=MaterialFunctionCall() {materialFunction=&Scatter_InstanceCoords}.Result)
  n18 = TextureSampleParameter2D(UVs=@n17.PS) {parameterName=PositionRT, texture=&Black}
  n19 = Frac(None=Divide(
    A=ComponentMask(None=@n18.RGB) {g=true, r=true}
    B=CollectionParameter() {collection=&MPC_Erosion, parameterName=GridSize}
   ))
  Add(
   A=AppendVector(
    A=MaterialFunctionCall!(Vector2=@n19, MaximumNumberOfDigits, UVs=@n16, DebugTextLocation RG_UpperRight BA_LowerLeft, Component Spacing) {materialFunction=/Engine/Functions/Engine_MaterialFunctions02/Utility/DebugFloat2Values}.ColorCodedOutput
    B=Constant()
   )
   B=VectorParameter() {parameterName=asdfasfsadf, defaultValue=<rgba 0 0 1 0>}.RGB
  )
  n20 = MaterialFunctionCall() {materialFunction=/Engine/Functions/Engine_MaterialFunctions02/Utility/VectorLength}
  n21 = CollectionParameter() {collection=&MPC_Erosion, parameterName=ResX}
  Multiply(
   A=CollectionParameter() {collection=&MPC_Erosion, parameterName=Solubility}
   B=Divide(
    A=Divide(A=@n20.V2 Length, B=@n21)
    B=CollectionParameter() {collection=&MPC_Erosion, parameterName=GridSize}
   )
  )
  n22 = TextureCoordinate()
  n23 = Ceil(None=Subtract(A=OneMinus(None=@n22)) {constB=0.5})
  n24 = MaterialFunctionCall(Float2=LinearInterpolate(A=@n19, B=OneMinus(None=@n19), Alpha=Saturate(None=@n23))) {materialFunction=&BreakOutFloat2Components}
  n25 = Multiply(A=@n24.R, B=@n24.R)
  n26 = Multiply(
   A=Divide(
    A=TextureSampleParameter2D(UVs=@n17.PS) {parameterName=VelocityRT, texture=&Black}.RGB
    B=ScalarParameter() {defaultValue=256, parameterName=Divisor}
   )
   B=ScalarParameter() {defaultValue=1, parameterName=Output}
  )
  n27 = TextureSampleParameter2D(UVs=@n17.PS) {parameterName=VelocityRT, texture=&Black}
  n28 = TextureSampleParameter2D(UVs=@n17.PS) {parameterName=VelocityRTPrev, texture=&Black}
  n29 = Multiply(A=@n18.A) {constB=1000}
  n30 = Multiply(A=Multiply(A=Subtract(A=@n27.A, B=@n28.A) {desc="Positive = Saturation Increase"}, B=@n29)) {constB=-1}
  n31 = CollectionParameter() {collection=&MPC_Erosion, parameterName=Evaporation}
  n32 = CollectionParameter() {collection=&MPC_Erosion, parameterName=deltaT}
  n33 = Multiply(A=Multiply(A=@n28.A, B=Multiply(A=@n31, B=@n32)), B=@n29)
  Multiply(
   A=LinearInterpolate(B=Saturate(None=@n25), Alpha=ScalarParameter() {defaultValue=1, parameterName=Quantize}) {constA=1}
   B=AppendVector(A=ComponentMask(None=@n26) {g=true, r=true}, B=Add(A=@n30, B=@n33))
  )
  ScalarParameter() {defaultValue=64, desc=water amt, parameterName=n1}
  ScalarParameter() {defaultValue=0.01, parameterName=FakeSedimentTransfer}
  ScalarParameter() {defaultValue=1, parameterName=a}
  TextureCoordinate() {coordinateIndex=1}
 ]
}
