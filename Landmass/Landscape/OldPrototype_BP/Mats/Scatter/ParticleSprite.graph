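# ParticleSprite is a prototype landscape scatter material that computes sprite quad placement from a position texture, with a `watersize` parameter suggesting water particles.
# It samples PositionRT with UV channel 1 and subtracts world position (excluding shader offsets), giving an offset that collapses vertices onto the stored point.
# Centered, rescaled UV0 times `watersize` spreads corners into a flat XY quad.
# A dark blue constant is also present.
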Material /Landmass/Landscape/OldPrototype_BP/Mats/Scatter/ParticleSprite
graph {
 properties {pixelDepthOffsetMode=PDOM_Legacy, usedShadingModels=Default Lit}
 nodes [
  n1 = WorldPosition() {worldPositionShaderOffset=WPT_ExcludeAllShaderOffsets}
  n2 = TextureCoordinate() {coordinateIndex=1}
  n3 = ConstantBiasScale(None=TextureCoordinate()) {bias=-0.5, scale=2}
  n4 = ScalarParameter() {defaultValue=8, parameterName=watersize}
  Add(
   A=Add(A=Multiply(A=@n1.XYZ) {constB=-1}, B=TextureSampleParameter2D(UVs=@n2) {parameterName=PositionRT, texture=&Black}.RGB)
   B=AppendVector(A=Multiply(A=@n3, B=@n4), B=Constant())
  )
  Constant3Vector() {constant=<rgba 0 0 0.2 0>}
 ]
}
