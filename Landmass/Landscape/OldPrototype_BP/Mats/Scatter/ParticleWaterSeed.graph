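# Unlit alpha-composite material from the old Landmass blueprint prototype that, judging by its name, seeds water particles for erosion.
# Scales UVs by the MPC_Erosion resolution and floors them, feeding per-cell vector noise that is thresholded into a random seed mask.
# Multiplies that mask by normalized StartWater and inverts it.
# A separate branch builds a vector from the scaled HeightRT sample plus a Z offset; negative emissive is allowed.
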Material /Landmass/Landscape/OldPrototype_BP/Mats/Scatter/ParticleWaterSeed
graph {
 properties {
  bAllowNegativeEmissiveColor=true
  blendMode=BLEND_AlphaComposite
  pixelDepthOffsetMode=PDOM_Legacy
  shadingModel=MSM_Unlit
  usedShadingModels=Unlit
 }
 nodes [
  n1 = TextureSampleParameter2D() {parameterName=HeightRT, texture=&Black}
  n2 = CollectionParameter() {collection=/Landmass/Landscape/BlueprintBrushes/MPC/MPC_Erosion, parameterName=ZScale}
  AppendVector(A=Constant2Vector(), B=Add(A=ScalarParameter() {defaultValue=100, parameterName=Z}, B=Multiply(A=@n1.R, B=@n2)))
  n3 = Multiply(
   A=TextureCoordinate()
   B=AppendVector(
    A=CollectionParameter() {collection=&MPC_Erosion, parameterName=ResX}
    B=CollectionParameter() {collection=&MPC_Erosion, parameterName=ResY}
   )
  )
  n4 = VectorNoise(World Position=AppendVector(A=Floor(None=@n3), B=Constant()))
  n5 = Ceil(None=Subtract(A=ComponentMask(None=@n4) {r=true}, B=ScalarParameter() {parameterName=RandomThreshold}) {constB=0.99})
  n6 = Divide(A=CollectionParameter() {collection=&MPC_Erosion, parameterName=StartWater}) {constB=1000, desc=Max Water}
  OneMinus(None=Multiply(A=Saturate(None=@n5), B=Saturate(None=@n6)))
  VectorParameter() {parameterName=P}
 ]
}
