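# Unlit prototype Landmass scatter material that computes a particle position update from render-target textures, using MPC_Erosion parameters.
# A Custom HLSL node advances PositionRT by VelocityRT times deltaT, then samples normal and height at the displaced UV.
# Positions under or near the surface get pushed along the reconstructed normal.
# Extra nodes compute evaporation on alpha and edge-wrapped XY, but no material output connection is shown.
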Material /Landmass/Landscape/OldPrototype_BP/Mats/Scatter/ScatterPosition
graph {
 properties {
  bAllowNegativeEmissiveColor=true
  blendMode=BLEND_AlphaComposite
  pixelDepthOffsetMode=PDOM_Legacy
  shadingModel=MSM_Unlit
  usedShadingModels=Unlit
 }
 nodes [
  MaterialExpressionTextureObjectParameter_1 = TextureObjectParameter() {parameterName=PositionRT, texture=&Black}
  MaterialExpressionTextureObjectParameter_5 = TextureObjectParameter() {parameterName=VelocityRT, texture=&Black}
  MaterialExpressionTextureObjectParameter_3 = TextureObjectParameter() {parameterName=NormalRT, texture=&Black}
  MaterialExpressionTextureObjectParameter_2 = TextureObjectParameter() {parameterName=HeightRT, texture=&Black}
  MaterialExpressionCollectionParameter_15 = CollectionParameter() {collection=/Landmass/Landscape/BlueprintBrushes/MPC/MPC_Erosion, parameterName=deltaT}
  MaterialExpressionMultiply_14 = Multiply(
   A=AppendVector(
    A=CollectionParameter() {collection=&MPC_Erosion, parameterName=ResX}
    B=CollectionParameter() {collection=&MPC_Erosion, parameterName=ResY}
   )
   B=CollectionParameter() {collection=&MPC_Erosion, parameterName=GridSize}
  )
  MaterialExpressionTextureCoordinate_1 = TextureCoordinate()
  n1 = Custom!(
   PositionRT=@MaterialExpressionTextureObjectParameter_1
   VelocityRT=@MaterialExpressionTextureObjectParameter_5
   NormalRT=@MaterialExpressionTextureObjectParameter_3
   HeightRT=@MaterialExpressionTextureObjectParameter_2
   deltaT=@MaterialExpressionCollectionParameter_15
   TotalSize=@MaterialExpressionMultiply_14
   inUV=@MaterialExpressionTextureCoordinate_1
   inUV1=@MaterialExpressionTextureCoordinate_1
  ) {
   code="//PositionRT: input Texture Object containing \r\n//VelcotiyRT: input \r\n//NormalRT: input \r\n//deltaT: input\r\n//input TextureSize\r\n//input GridSize\r\n\r\nfloat4 position = PositionRT.SampleLevel(PositionRTSampler, inUV, 0).rgba;\r\nfloat3 velocity = VelocityRT.SampleLevel(VelocityRTSampler, inUV, 0).rgb;\r\n\r\nfloat2 UV1 = inUV1 + (position.rg / TotalSize);\r\nfloat startZ = NormalRT.SampleLevel(NormalRTSampler, UV1, 0).b;\r\n\r\nposition.rgb += (velocity * deltaT);\r\n\r\nUV1 = inUV1 + (position.rg / TotalSize);\r\nfloat3 endnormal = NormalRT.SampleLevel(NormalRTSampler, UV1, 0).rgb;\r\n\r\nfloat endZ = HeightRT.SampleLevel(NormalRTSampler, UV1, 0).r;\r\n\r\nendnormal.rg = (endnormal.rg - 0.5) * 2;\r\nfloat3 normal = normalize( float3( endnormal.rg, sqrt(1 - dot(endnormal.xy, endnormal.xy) ) ) );\r\n\r\n//position.z += startZ - endZ.b;\r\n\r\nif( position.z - endZ < 50 )\r\n{\r\n\r\n\t//position += velocity * deltaT * (position.z - 10);\r\n\tposition.rgb += normal * 10;\r\n\t//position.z = max(endZ.b + 10, position.z);\r\n}\r\n\r\nreturn position;"
   description=Erosion Pass
   outputType=CMOT_Float4
   inputs [
    {
     inputName=PositionRT
     input {expression=&self:MaterialExpressionTextureObjectParameter_1}
    }
    {
     inputName=VelocityRT
     input {expression=&self:MaterialExpressionTextureObjectParameter_5}
    }
    {
     inputName=NormalRT
     input {expression=&self:MaterialExpressionTextureObjectParameter_3}
    }
    {
     inputName=HeightRT
     input {expression=&self:MaterialExpressionTextureObjectParameter_2}
    }
    {
     inputName=deltaT
     input {expression=&self:MaterialExpressionCollectionParameter_15}
    }
    {inputName=TotalSize, input {expression=&self:MaterialExpressionMultiply_14}}
    {inputName=inUV, input {expression=&self:MaterialExpressionTextureCoordinate_1}}
    {inputName=inUV1, input {expression=&self:MaterialExpressionTextureCoordinate_1}}
   ]
  }
  n2 = ComponentMask(None=@n1) {g=true, r=true}
  AppendVector(A=@n2, B=ComponentMask(None=@n1) {b=true})
  n3 = ComponentMask(None=@n1) {a=true}
  n4 = CollectionParameter() {collection=&MPC_Erosion, parameterName=Evaporation}
  n5 = CollectionParameter() {collection=&MPC_Erosion, parameterName=deltaT}
  n6 = OneMinus(None=Saturate(None=Multiply(A=@n4, B=@n5)))
  OneMinus(None=Saturate(None=Multiply(A=@n3, B=@n6)))
  n7 = AppendVector(
   A=CollectionParameter() {collection=&MPC_Erosion, parameterName=TotalSizeX}
   B=CollectionParameter() {collection=&MPC_Erosion, parameterName=TotalSizeY}
  )
  n8 = Multiply(A=Subtract(A=TextureCoordinate()) {constB=0.5}, B=@n7)
  n9 = Add(A=@n8, B=@n2)
  n10 = Frac(None=Add(A=Divide(A=@n9, B=@n7)) {constB=0.5})
  Subtract(A=Multiply(A=@n7, B=Subtract(A=@n10) {constB=0.5}), B=@n8)
  TextureCoordinate() {coordinateIndex=1}
 ]
}
