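# Unlit prototype landscape material for the velocity step of a render-target scatter simulation, built on erosion logic.
# A custom HLSL node reads position, velocity, height and normal targets, adds gravity and damping, and removes into-surface velocity near terrain.
# Further branches compute sediment dissolve/deposit with slope and edge detection, plus an f16-pair packed position update.
# Timestep, resolution and damping come from MPC_Erosion; no connection to a material output appears.
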Material /Landmass/Landscape/OldPrototype_BP/Mats/Scatter/ScatterVelocity
graph {
 properties {
  bAllowNegativeEmissiveColor=true
  blendMode=BLEND_AlphaComposite
  pixelDepthOffsetMode=PDOM_Legacy
  shadingModel=MSM_Unlit
  usedShadingModels=Unlit
 }
 nodes [
  n1 = TextureCoordinate() {coordinateIndex=1}
  n2 = AppendVector(A=ScalarParameter() {defaultValue=2048, parameterName=SizeX}, B=ScalarParameter() {defaultValue=2048, parameterName=SizeY})
  n3 = ScalarParameter() {defaultValue=16, parameterName=GridSize}
  AppendVector(A=Multiply(A=ConstantBiasScale(None=@n1) {bias=-0.5, scale=1}, B=Multiply(A=@n2, B=@n3)), B=Constant())
  Constant()
  n4 = MaterialFunctionCall() {materialFunction=/Landmass/Landscape/OldPrototype_BP/Mats/MF/PointSampleUVs}
  n5 = Constant()
  n6 = TextureSampleParameter2D!(UVs=@n4.TexelCenterUVs, Level=@n5, Apply View MipBias) {mipValueMode=TMVM_MipLevel, parameterName=PositionRT, texture=&Black}
  MaterialExpressionAppendVector_0 = AppendVector(A=@n6.RGB, B=@n6.RGB)
  MaterialExpressionTextureSampleParameter2D_7 = TextureSampleParameter2D(UVs=@n4.TexelCenterUVs, Level=@n5) {mipValueMode=TMVM_MipLevel, parameterName=VelocityRT, texture=&Black}
  MaterialExpressionTextureObjectParameter_3 = TextureObjectParameter() {parameterName=NormalRT, texture=&Black}
  MaterialExpressionTextureObjectParameter_0 = TextureObjectParameter() {parameterName=HeightRT, texture=&Black}
  MaterialExpressionCollectionParameter_3 = CollectionParameter() {collection=/Landmass/Landscape/BlueprintBrushes/MPC/MPC_Erosion, parameterName=deltaT}
  MaterialExpressionMultiply_14 = Multiply(
   A=AppendVector(
    A=CollectionParameter() {collection=&MPC_Erosion, parameterName=ResX}
    B=CollectionParameter() {collection=&MPC_Erosion, parameterName=ResY}
   )
   B=CollectionParameter() {collection=&MPC_Erosion, parameterName=GridSize}
  )
  MaterialExpressionTextureCoordinate_1 = TextureCoordinate()
  MaterialExpressionCollectionParameter_10 = CollectionParameter() {collection=&MPC_Erosion, parameterName=VelocityDamping}
  n7 = Custom!(
   Position=@MaterialExpressionAppendVector_0
   Velocity=@MaterialExpressionTextureSampleParameter2D_7.RGB
   NormalRT=@MaterialExpressionTextureObjectParameter_3
   HeightRT=@MaterialExpressionTextureObjectParameter_0
   deltaT=@MaterialExpressionCollectionParameter_3
   WorldSize=@MaterialExpressionMultiply_14
   inUV=@MaterialExpressionTextureCoordinate_1
   Damping=@MaterialExpressionCollectionParameter_10
  ) {
   code="//Position: input Texture Object containing \r\n//Velocity: input \r\n//NormalRT: input \r\n//HeightRT: input\r\n//deltaT: input\r\n//input TextureSize\r\n//input GridSize\r\n\r\nfloat3 accel = float3(0,0, -980 * deltaT);\r\nfloat3 pos = Position;\r\n\r\nfloat2 UV = inUV + (Position / WorldSize);\r\n\r\nfloat startZ = HeightRT.SampleLevel(HeightRTSampler, UV, 0).b;\r\n\r\nfloat3 velocity = 0;\r\n\r\nif(Position.a > 0)\r\n{\r\n\tvelocity = Velocity + accel;\r\n\tvelocity *= exp(-Damping*deltaT);\r\n    //velocity *= (1-Damping);\r\n}\r\n\r\n\r\nfloat2 normal2d = NormalRT.SampleLevel(NormalRTSampler, UV, 0).ba;\r\nnormal2d = (normal2d - 0.5) * 2;\r\n\r\nfloat terrainZ = HeightRT.SampleLevel(HeightRTSampler, UV, 0).r;\r\nfloat3 normal = normalize( float3( normal2d, sqrt(1 - dot(normal2d , normal2d ) ) ) );\r\n\r\nif((velocity.z * deltaT) + pos.z < terrainZ + 50 &&&& Position.a > 0)\r\n{\r\n\t//float boost = saturate(-1 * (velocity.z* deltaT + pos.z -16));\r\n\tvelocity -= normal * min(0, dot(normal, velocity)) ;\r\n}\r\n\r\nreturn velocity;\r\n\r\n"
   description=Velocity Acceleration
   inputs [
    {inputName=Position, input {expression=&self:MaterialExpressionAppendVector_0}}
    {
     inputName=Velocity
     input {
      expression=&self:MaterialExpressionTextureSampleParameter2D_7
      mask=1
      maskB=1
      maskG=1
      maskR=1
     }
    }
    {
     inputName=NormalRT
     input {expression=&self:MaterialExpressionTextureObjectParameter_3}
    }
    {
     inputName=HeightRT
     input {expression=&self:MaterialExpressionTextureObjectParameter_0}
    }
    {
     inputName=deltaT
     input {expression=&self:MaterialExpressionCollectionParameter_3}
    }
    {inputName=WorldSize, input {expression=&self:MaterialExpressionMultiply_14}}
    {inputName=inUV, input {expression=&self:MaterialExpressionTextureCoordinate_1}}
    {
     inputName=Damping
     input {expression=&self:MaterialExpressionCollectionParameter_10}
    }
   ]
  }
  n8 = Divide(
   A=MaterialFunctionCall(Vector 2=@n7) {materialFunction=/Engine/Functions/Engine_MaterialFunctions02/Utility/VectorLength}.V2 Length
   B=CollectionParameter() {collection=&MPC_Erosion, parameterName=ResX}
  )
  Divide(A=@n8, B=CollectionParameter() {collection=&MPC_Erosion, parameterName=GridSize})
  MaterialExpressionTextureSampleParameter2D_2 = TextureSampleParameter2D!(UVs, Apply View MipBias) {parameterName=VelocityRT, texture=&Black}
  n9 = Custom!(N=@MaterialExpressionTextureSampleParameter2D_2.R) {
   code="//extracts 2 16bit floats from a  32bit float\r\n\r\nuint f16 = asuint(N);\r\nf16 = (f16 && 0xffff0000) >> 16;\r\nuint sign = (f16 && 0x8000)<<16;\r\nuint mant = (f16 && 0x3ffu) <<13; // 13 bits are zeros\r\nuint exp =  (f16&&0x7fff)==0 ? 0 : (((((const uint) f16)>>10)&&0x1f));\r\nf16 =  sign | (exp -15+127) << 23 | mant;\r\nfloat outval1 = asfloat( f16 );\r\n\r\nf16 = asuint(N);\r\nf16 = (f16 && 0xffff);\r\nsign = (f16 && 0x8000)<<16;\r\nmant = (f16 && 0x3ffu) <<13; // 13 bits are zeros\r\nexp =  (f16&&0x7fff)==0 ? 0 : (((((const uint) f16)>>10)&&0x1f));\r\nf16 =  sign | (exp -15+127) << 23 | mant;\r\nfloat outval2 = asfloat( f16 );\r\n\r\nreturn float2(outval1, outval2);\r\n"
   description=Unpack 2 f16s from f32
   outputType=CMOT_Float2
   inputs [{
     inputName=N
     input {
      expression=&self:MaterialExpressionTextureSampleParameter2D_2
      mask=1
      maskR=1
      outputIndex=1
     }
    }]
  }
  n10 = MaterialFunctionCall(Float2=@n9) {materialFunction=&BreakOutFloat2Components}
  n11 = Custom(N=@MaterialExpressionTextureSampleParameter2D_2.G) {
   code="//extracts 2 16bit floats from a  32bit float\r\n\r\nuint f16 = asuint(N);\r\nf16 = (f16 && 0xffff0000) >> 16;\r\nuint sign = (f16 && 0x8000)<<16;\r\nuint mant = (f16 && 0x3ffu) <<13; // 13 bits are zeros\r\nuint exp =  (f16&&0x7fff)==0 ? 0 : (((((const uint) f16)>>10)&&0x1f));\r\nf16 =  sign | (exp -15+127) << 23 | mant;\r\nfloat outval1 = asfloat( f16 );\r\n\r\nf16 = asuint(N);\r\nf16 = (f16 && 0xffff);\r\nsign = (f16 && 0x8000)<<16;\r\nmant = (f16 && 0x3ffu) <<13; // 13 bits are zeros\r\nexp =  (f16&&0x7fff)==0 ? 0 : (((((const uint) f16)>>10)&&0x1f));\r\nf16 =  sign | (exp -15+127) << 23 | mant;\r\nfloat outval2 = asfloat( f16 );\r\n\r\nreturn float2(outval1, outval2);\r\n"
   description=Unpack 2 f16s from f32
   outputType=CMOT_Float2
   inputs [{
     inputName=N
     input {
      expression=&self:MaterialExpressionTextureSampleParameter2D_2
      mask=1
      maskG=1
      outputIndex=2
     }
    }]
  }
  n12 = MaterialFunctionCall(Float2=@n11) {materialFunction=&BreakOutFloat2Components}
  n13 = Custom(N=@MaterialExpressionTextureSampleParameter2D_2.B) {
   code="//extracts 2 16bit floats from a  32bit float\r\n\r\nuint f16 = asuint(N);\r\nf16 = (f16 && 0xffff0000) >> 16;\r\nuint sign = (f16 && 0x8000)<<16;\r\nuint mant = (f16 && 0x3ffu) <<13; // 13 bits are zeros\r\nuint exp =  (f16&&0x7fff)==0 ? 0 : (((((const uint) f16)>>10)&&0x1f));\r\nf16 =  sign | (exp -15+127) << 23 | mant;\r\nfloat outval1 = asfloat( f16 );\r\n\r\nf16 = asuint(N);\r\nf16 = (f16 && 0xffff);\r\nsign = (f16 && 0x8000)<<16;\r\nmant = (f16 && 0x3ffu) <<13; // 13 bits are zeros\r\nexp =  (f16&&0x7fff)==0 ? 0 : (((((const uint) f16)>>10)&&0x1f));\r\nf16 =  sign | (exp -15+127) << 23 | mant;\r\nfloat outval2 = asfloat( f16 );\r\n\r\nreturn float2(outval1, outval2);\r\n"
   description=Unpack 2 f16s from f32
   outputType=CMOT_Float2
   inputs [{
     inputName=N
     input {
      expression=&self:MaterialExpressionTextureSampleParameter2D_2
      mask=1
      maskB=1
      outputIndex=3
     }
    }]
  }
  n14 = MaterialFunctionCall(Float2=@n13) {materialFunction=&BreakOutFloat2Components}
  MaterialFunctionCall(X=@n10.R, Y=@n12.R, Z=@n14.R) {desc=Velocity, materialFunction=&MakeFloat3}
  MaterialExpressionMultiply_3 = Multiply() {desc=Velocity}
  MaterialExpressionMaterialFunctionCall_0 = MaterialFunctionCall(Float3=@MaterialExpressionMultiply_3) {materialFunction=&BreakOutFloat3Components}
  MaterialExpressionAdd_10 = Add(A=MaterialFunctionCall(X=@n10.G, Y=@n12.G, Z=@n14.G) {desc=Position, materialFunction=&MakeFloat3}.Result) {constB=0}
  MaterialExpressionTextureObjectParameter_6 = TextureObjectParameter() {parameterName=NormalRT, texture=&Black}
  MaterialExpressionScalarParameter_18 = ScalarParameter() {defaultValue=0.05, parameterName=deltaT}
  MaterialExpressionAppendVector_4 = AppendVector(A=ScalarParameter() {defaultValue=1024, parameterName=SizeX}, B=ScalarParameter() {defaultValue=1024, parameterName=SizeY})
  MaterialExpressionScalarParameter_17 = ScalarParameter() {defaultValue=16, parameterName=GridSize}
  MaterialExpressionTextureCoordinate_4 = TextureCoordinate()
  n15 = Custom!(
   PositionRT=@MaterialExpressionAdd_10
   VelocityRT=@MaterialExpressionMultiply_3
   NormalRT=@MaterialExpressionTextureObjectParameter_6
   deltaT=@MaterialExpressionScalarParameter_18
   TextureSize=@MaterialExpressionAppendVector_4
   GridSize=@MaterialExpressionScalarParameter_17
   UV1=@MaterialExpressionTextureCoordinate_4
  ) {
   code="//PositionRT: input Texture Object containing \r\n//VelcotiyRT: input \r\n//NormalRT: input \r\n//deltaT: input\r\n//input TextureSize\r\n//input GridSize\r\n\r\nfloat3 position = PositionRT;\r\nfloat3 velocity = VelocityRT;\r\n\r\nposition += (velocity * deltaT);\r\n\r\nfloat2 UV = position.rg;\r\nUV = ((UV / TextureSize) / GridSize) + 0.5;\r\n\r\nfloat startZ = NormalRT.SampleLevel(NormalRTSampler, UV1 + (PositionRT / (GridSize * TextureSize)),0);\r\nfloat curZ = NormalRT.SampleLevel(NormalRTSampler, UV1 + (position / (GridSize * TextureSize)),0);\r\n\r\n//float terrainZ = NormalRT.SampleLevel(NormalRTSampler, UV, 0).b;\r\n\r\n//position.z = max(terrainZ+10, position.z);\r\n\r\nposition.z += startZ - curZ;\r\nposition.z = max(10, position.z);\r\n\r\nreturn position;"
   description=Position Update
   inputs [
    {inputName=PositionRT, input {expression=&self:MaterialExpressionAdd_10}}
    {inputName=VelocityRT, input {expression=&self:MaterialExpressionMultiply_3}}
    {
     inputName=NormalRT
     input {expression=&self:MaterialExpressionTextureObjectParameter_6}
    }
    {inputName=deltaT, input {expression=&self:MaterialExpressionScalarParameter_18}}
    {inputName=TextureSize, input {expression=&self:MaterialExpressionAppendVector_4}}
    {inputName=GridSize, input {expression=&self:MaterialExpressionScalarParameter_17}}
    {inputName=UV1, input {expression=&self:MaterialExpressionTextureCoordinate_4}}
   ]
  }
  MaterialExpressionMaterialFunctionCall_1 = MaterialFunctionCall(Float3=@n15) {materialFunction=&BreakOutFloat3Components}
  n16 = Custom!(N1=@MaterialExpressionMaterialFunctionCall_0.R, N2=@MaterialExpressionMaterialFunctionCall_1.R) {
   code="uint uRes32 = asuint(N1);\r\nint  iExp      = (const int)(((const int) ((uRes32>>23)&&0xff))-127+15);\r\nuint  expClamp = iExp < 0 ? 0 : (iExp>31 ? 31 : (uint)iExp);\r\nuint sign = (( uRes32>>16 ) && 0x8000);\r\nuint mant = ((uRes32>>13)&&0x3ff);\r\nuint exp  = (expClamp << 10);\r\nuint bits1 = sign | exp | mant; \r\nbits1 = bits1 && 0xffff;\r\nbits1 = bits1 << 16;\r\n\r\nuRes32 = asuint(N2);\r\niExp      = (const int)(((const int) ((uRes32>>23)&&0xff))-127+15);\r\nexpClamp = iExp < 0 ? 0 : (iExp>31 ? 31 : (uint)iExp);\r\nsign = (( uRes32>>16 ) && 0x8000);\r\nmant = ((uRes32>>13)&&0x3ff);\r\nexp  = (expClamp << 10);\r\nuint bits2 = (sign | exp | mant) && 0xffff;\r\n\r\nreturn asfloat( bits1 | bits2 );\r\n"
   description=Pack 2 f16s into f32
   outputType=CMOT_Float1
   inputs [
    {inputName=N1, input {expression=&self:MaterialExpressionMaterialFunctionCall_0}}
    {inputName=N2, input {expression=&self:MaterialExpressionMaterialFunctionCall_1}}
   ]
  }
  n17 = Custom(N1=@MaterialExpressionMaterialFunctionCall_0.G, N2=@MaterialExpressionMaterialFunctionCall_1.G) {
   code="uint uRes32 = asuint(N1);\r\nint  iExp      = (const int)(((const int) ((uRes32>>23)&&0xff))-127+15);\r\nuint  expClamp = iExp < 0 ? 0 : (iExp>31 ? 31 : (uint)iExp);\r\nuint sign = (( uRes32>>16 ) && 0x8000);\r\nuint mant = ((uRes32>>13)&&0x3ff);\r\nuint exp  = (expClamp << 10);\r\nuint bits1 = sign | exp | mant; \r\nbits1 = bits1 && 0xffff;\r\nbits1 = bits1 << 16;\r\n\r\nuRes32 = asuint(N2);\r\niExp      = (const int)(((const int) ((uRes32>>23)&&0xff))-127+15);\r\nexpClamp = iExp < 0 ? 0 : (iExp>31 ? 31 : (uint)iExp);\r\nsign = (( uRes32>>16 ) && 0x8000);\r\nmant = ((uRes32>>13)&&0x3ff);\r\nexp  = (expClamp << 10);\r\nuint bits2 = (sign | exp | mant) && 0xffff;\r\n\r\nreturn asfloat( bits1 | bits2 );\r\n"
   description=Pack 2 f16s into f32
   outputType=CMOT_Float1
   inputs [
    {
     inputName=N1
     input {expression=&self:MaterialExpressionMaterialFunctionCall_0, outputIndex=1}
    }
    {
     inputName=N2
     input {expression=&self:MaterialExpressionMaterialFunctionCall_1, outputIndex=1}
    }
   ]
  }
  n18 = Custom(N1=@MaterialExpressionMaterialFunctionCall_0.B, N2=@MaterialExpressionMaterialFunctionCall_1.B) {
   code="uint uRes32 = asuint(N1);\r\nint  iExp      = (const int)(((const int) ((uRes32>>23)&&0xff))-127+15);\r\nuint  expClamp = iExp < 0 ? 0 : (iExp>31 ? 31 : (uint)iExp);\r\nuint sign = (( uRes32>>16 ) && 0x8000);\r\nuint mant = ((uRes32>>13)&&0x3ff);\r\nuint exp  = (expClamp << 10);\r\nuint bits1 = sign | exp | mant; \r\nbits1 = bits1 && 0xffff;\r\nbits1 = bits1 << 16;\r\n\r\nuRes32 = asuint(N2);\r\niExp      = (const int)(((const int) ((uRes32>>23)&&0xff))-127+15);\r\nexpClamp = iExp < 0 ? 0 : (iExp>31 ? 31 : (uint)iExp);\r\nsign = (( uRes32>>16 ) && 0x8000);\r\nmant = ((uRes32>>13)&&0x3ff);\r\nexp  = (expClamp << 10);\r\nuint bits2 = (sign | exp | mant) && 0xffff;\r\n\r\nreturn asfloat( bits1 | bits2 );\r\n"
   description=Pack 2 f16s into f32
   outputType=CMOT_Float1
   inputs [
    {
     inputName=N1
     input {expression=&self:MaterialExpressionMaterialFunctionCall_0, outputIndex=2}
    }
    {
     inputName=N2
     input {expression=&self:MaterialExpressionMaterialFunctionCall_1, outputIndex=2}
    }
   ]
  }
  MaterialFunctionCall(X=@n16, Y=@n17, Z=@n18) {desc=Position, materialFunction=&MakeFloat3}
  Multiply(A=@n7) {desc=New Velocity}
  MaterialExpressionAdd_0 = Add(
   A=Multiply(A=@MaterialExpressionTextureSampleParameter2D_7.RGB, B=@MaterialExpressionCollectionParameter_3)
   B=TextureSampleParameter2D!(
    UVs=MaterialFunctionCall() {materialFunction=&PointSampleUVs}.TexelCenterUVs
    Level=Constant()
    Apply View MipBias
   ) {mipValueMode=TMVM_MipLevel, parameterName=PositionRT, texture=&Black}.RGB
  )
  n19 = ComponentMask(None=@MaterialExpressionAdd_0) {g=true, r=true}
  MaterialExpressionAppendVector_1 = AppendVector(
   A=CollectionParameter() {collection=&MPC_Erosion, parameterName=ResX}
   B=CollectionParameter() {collection=&MPC_Erosion, parameterName=ResY}
  )
  MaterialExpressionMultiply_8 = Multiply(
   A=@MaterialExpressionAppendVector_1
   B=CollectionParameter() {collection=&MPC_Erosion, parameterName=GridSize}
  )
  n20 = TextureSampleParameter2D(
   UVs=Add(
    A=MaterialFunctionCall() {materialFunction=&PointSampleUVs}.TexelCenterUVs
    B=Divide(A=@n19, B=@MaterialExpressionMultiply_8)
   )
   Level=Constant()
  ) {mipValueMode=TMVM_MipLevel, parameterName=NormalRT, texture=&Black}
  n21 = ComponentMask(None=DeriveNormalZ(InXY=ComponentMask(None=@n20.RGB) {g=true, r=true})) {b=true}
  n22 = Multiply(A=@n21, B=@n21)
  Multiply(A=@n8, B=SquareRoot(None=OneMinus(None=@n22)))
  n23 = TextureSampleParameter2D!(UVs, Apply View MipBias) {parameterName=NormalRT, texture=&Black}
  n24 = ComponentMask(None=DeriveNormalZ(InXY=ComponentMask(None=@n23.RGB) {g=true, r=true})) {b=true}
  n25 = Multiply(A=@n24, B=@n24)
  n26 = CollectionParameter() {collection=&MPC_Erosion, parameterName=Solubility}
  Multiply(A=SquareRoot(None=OneMinus(None=@n25)), B=@n26)
  n27 = Saturate(None=Multiply(A=@n26, B=@n8))
  Multiply(A=@n27) {desc=Prevent Holes}
  n28 = Multiply(A=TextureSampleParameter2D!(
    UVs=MaterialFunctionCall() {materialFunction=&PointSampleUVs}.TexelCenterUVs
    Level=Constant()
    Apply View MipBias
   ) {mipValueMode=TMVM_MipLevel, parameterName=VelocityRT, texture=&Black}.A) {desc="v.A = Sediment per Water Unit"}
  n29 = MaterialFunctionCall() {materialFunction=&PointSampleUVs}
  n30 = Constant()
  n31 = Subtract(
   A=Multiply(A=TextureSampleParameter2D(UVs=@n29.TexelCenterUVs, Level=@n30) {mipValueMode=TMVM_MipLevel, parameterName=VelocityRT, texture=&Black}.A) {desc="v.A = Sediment per Water Unit"}
   B=@n27
  ) {desc=Positive if Over-Saturated}
  n32 = CollectionParameter() {collection=&MPC_Erosion, parameterName=deltaT}
  n33 = Multiply(A=ScalarParameter() {defaultValue=0.1, parameterName=Dissolving Constant}, B=@n32)
  MaterialExpressionTextureObjectParameter_2 = TextureObjectParameter() {parameterName=HeightRT, texture=&Black}
  MaterialExpressionTextureCoordinate_0 = TextureCoordinate()
  MaterialExpressionTextureObjectParameter_1 = TextureObjectParameter() {parameterName=VelocityRT, texture=&Black}
  n34 = Custom!(
   Tex=@MaterialExpressionTextureObjectParameter_2
   UV=@MaterialExpressionTextureCoordinate_0
   TextureSize=@MaterialExpressionAppendVector_1
   WorldSize=@MaterialExpressionMultiply_8
   Position=@MaterialExpressionAdd_0
   VelocityRT=@MaterialExpressionTextureObjectParameter_1
  ) {
   code="float2 texelsize = 1/TextureSize;\r\n\r\nfloat2 offsets[8] = {float2(1,0) , float2(-1,0) , float2(0,1) , float2 (0,-1), float2(1,1), float2(1,-1), float2(-1,1), float2(-1,-1)};\r\n\r\nfloat minz = 1000000;\r\nfloat maxz = -1000000;\r\nfloat accum = 0;\r\n\r\nfloat2 startpos = UV + (Position.xy / WorldSize);\r\n\r\nfloat2 startUV = floor(startpos * TextureSize) + 0.5;\r\nstartUV /= TextureSize; \r\n\r\nfloat startZ = Tex.SampleLevel(TexSampler, startUV, 0);\r\n\r\nfor(int i = 0; i < 4; i++)\r\n{\r\n\tfloat2 curUV = startUV + (offsets[i] * texelsize);\r\n\tfloat offsetsample = Tex.SampleLevel(TexSampler, curUV, 0);\r\n\r\n\tminz = min(minz, offsetsample);\r\n\t//maxz = max(maxz, offsetsample);\r\n\taccum += offsetsample;\r\n\r\n}\r\n\r\nfloat2 velocity = normalize(VelocityRT.SampleLevel(VelocityRTSampler, startpos, 0).rg);\r\nfloat2 crossvel = float2(-velocity.y, velocity.x);\r\nfloat2 veloffsets[4] = {velocity, -velocity, crossvel, -crossvel};\r\n\r\nfor(int k = 0; k < 4; k++)\r\n{\r\n\tfloat2 curUV = startpos + (veloffsets[k] * texelsize);\r\n\tfloat offsetsample = Tex.SampleLevel(TexSampler, curUV, 0);\r\n\r\n\tmaxz = max(maxz, offsetsample);\r\n\r\n\r\n}\r\n\r\n\r\naccum /= 4;\r\n\r\nreturn float4(startZ, minz, maxz, accum);"
   description=Detect Edges
   outputType=CMOT_Float4
   inputs [
    {
     inputName=Tex
     input {expression=&self:MaterialExpressionTextureObjectParameter_2}
    }
    {inputName=UV, input {expression=&self:MaterialExpressionTextureCoordinate_0}}
    {inputName=TextureSize, input {expression=&self:MaterialExpressionAppendVector_1}}
    {inputName=WorldSize, input {expression=&self:MaterialExpressionMultiply_8}}
    {inputName=Position, input {expression=&self:MaterialExpressionAdd_0}}
    {
     inputName=VelocityRT
     input {expression=&self:MaterialExpressionTextureObjectParameter_1}
    }
   ]
  }
  n35 = MaterialFunctionCall(Float4=@n34) {materialFunction=&BreakOutFloat4Components}
  n36 = Saturate(None=Subtract(A=@n35.R, B=@n35.R) {desc=Center - Min Neighbor})
  n37 = CollectionParameter() {collection=&MPC_Erosion, parameterName=MaxAdvectSlope}
  n38 = Divide(
   A=Subtract(A=@n35.B, B=@n35.B) {desc=Max - Center}
   B=CollectionParameter() {collection=&MPC_Erosion, parameterName=GridSize}
  ) {desc=Slope}
  n39 = Saturate(None=Ceil(None=Subtract(A=@n37, B=@n38)))
  n40 = Subtract(A=@n28, B=Multiply(A=Multiply(A=@n31, B=@n33), B=Multiply(A=@n36, B=@n39)))
  n41 = ScalarParameter() {defaultValue=0.02, parameterName=Deposition Constant}
  n42 = Subtract(A=@n28, B=Multiply(A=@n31, B=Multiply(A=@n41, B=@n32)))
  n43 = Saturate(None=Ceil(None=@n31)) {desc=Extract Sign}
  OneMinus(None=Saturate(None=LinearInterpolate(A=@n40, B=@n42, Alpha=@n43) {desc="A = Undersaterated, B = Oversaturated"}))
  n44 = CollectionParameter() {collection=&MPC_Erosion, parameterName=Evaporation}
  n45 = CollectionParameter() {collection=&MPC_Erosion, parameterName=deltaT}
  OneMinus(None=Saturate(None=Multiply(A=@n44, B=@n45)))
  ScalarParameter() {defaultValue=0.9, parameterName=ssss}
  ScalarParameter() {defaultValue=1, parameterName=Min Slope}
  TextureCoordinate()
  TextureObject() {texture=&Black}
  TextureSampleParameter2D!(UVs, Apply View MipBias) {desc=Previous Velocity, parameterName=VelocityRT, texture=&Black}
 ]
}
