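# Unlit prototype material for Landmass landscape erosion that computes eroded height data from scene-capture, position, normal and height render targets.
# It builds a UV offset from the capture's red-green direction and clamped length, PositionRT scaled by Weathering, and NormalRT, using MPC_Erosion resolution and advection values.
# HeightRT is sampled at that offset; its red passes through SlopeSmoothing plus capture blue scaled by ApplySediment, and its green passes through unchanged.
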Material /Landmass/Landscape/OldPrototype_BP/Mats/Scatter/StartPositionErode
graph {
 properties {bAllowNegativeEmissiveColor=true, pixelDepthOffsetMode=PDOM_Legacy, shadingModel=MSM_Unlit, usedShadingModels=Unlit}
 nodes [
  n1 = MaterialFunctionCall() {materialFunction=/Landmass/Landscape/OldPrototype_BP/Mats/MF/PointSampleUVs}
  n2 = MaterialFunctionCall(Float2=@n1.TexelCenterUVs) {materialFunction=&BreakOutFloat2Components}
  AppendVector(A=@n2.R, B=OneMinus(None=@n2.G))
  n3 = MaterialFunctionCall() {materialFunction=&PointSampleUVs}
  n4 = TextureSampleParameter2D!(UVs=@n3.TexelCenterUVs, Level=Constant(), Apply View MipBias) {mipValueMode=TMVM_MipLevel, parameterName=SceneCaptureRT, texture=&Black}
  n5 = ScalarParameter() {defaultValue=1, parameterName=ApplySediment}
  n6 = MaterialFunctionCall() {materialFunction=&PointSampleUVs}
  n7 = TextureSampleParameter2D(UVs=@n1.TexelCenterUVs, Level=Constant()) {mipValueMode=TMVM_MipLevel, parameterName=SceneCaptureRT, texture=&Black}
  n8 = MaterialFunctionCall(Vector 3=@n7.RGB) {materialFunction=/Engine/Functions/Engine_MaterialFunctions02/Utility/VectorLength}
  n9 = ScalarParameter() {defaultValue=16, parameterName=GridSize}
  n10 = Divide(
   A=Min(B=Divide(A=@n8.V3 Length, B=@n9)) {constA=1}
   B=AppendVector(
    A=CollectionParameter() {collection=/Landmass/Landscape/BlueprintBrushes/MPC/MPC_Erosion, parameterName=ResX}
    B=CollectionParameter() {collection=&MPC_Erosion, parameterName=ResY}
   )
  )
  n11 = Normalize(VectorInput=Add(A=ComponentMask(None=@n7.RGB) {g=true, r=true}, B=Constant2Vector() {r=0}))
  n12 = MaterialFunctionCall() {materialFunction=&PointSampleUVs}
  n13 = Constant()
  n14 = Multiply(
   A=AppendVector(
    A=CollectionParameter() {collection=&MPC_Erosion, parameterName=ResX}
    B=CollectionParameter() {collection=&MPC_Erosion, parameterName=ResY}
   )
   B=CollectionParameter() {collection=&MPC_Erosion, parameterName=GridSize}
  )
  n15 = Divide(
   A=ComponentMask(None=TextureSampleParameter2D(UVs=@n12.TexelCenterUVs, Level=@n13) {mipValueMode=TMVM_MipLevel, parameterName=PositionRT, texture=&Black}.RGB) {g=true, r=true}
   B=@n14
  )
  n16 = Add(
   A=Multiply(
    A=Multiply(A=@n10, B=@n11)
    B=CollectionParameter() {collection=&MPC_Erosion, parameterName=AdvectionErosion}
   )
   B=Multiply(A=Multiply(A=@n15) {constB=0.01}, B=ScalarParameter() {parameterName=Weathering})
  )
  n17 = TextureSampleParameter2D(UVs=@n3.TexelCenterUVs, Level=Constant()) {mipValueMode=TMVM_MipLevel, parameterName=NormalRT, texture=&Black}
  n18 = Multiply(
   A=Divide(A=ComponentMask(None=@n17.RGB) {g=true, r=true}, B=@n14)
   B=CollectionParameter() {collection=&MPC_Erosion, parameterName=NormalAdvection}
  )
  n19 = TextureSampleParameter2D(UVs=Add(A=@n6.TexelCenterUVs, B=Add(A=@n16, B=@n18)), Level=Constant()) {mipValueMode=TMVM_MipLevel, parameterName=HeightRT, texture=&Black}
  MaterialFunctionCall(
   X=Add(
    A=Multiply(A=@n4.B, B=@n5)
    B=MaterialFunctionCall!(UVs=@n6.TexelCenterUVs, Height=@n19.R) {materialFunction=/Landmass/Landscape/OldPrototype_BP/Mats/MF/SlopeSmoothing}.Height
   )
   Y=@n19.G
  ) {materialFunction=&MakeFloat3}
  Multiply(A=@n8.V3 Length, B=@n8.V3 Length)
  Multiply(A=@n4.B) {constB=-1}
  ScalarParameter() {parameterName=Advection}
  TextureCoordinate()
 ]
}
